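; Blueprint function library of shared helpers for the Fortnite sidebar UI, mainly for opening confirmation and player profile panels.
; Navigation helpers build BP_NavigationPoint objects with StateTree transition tags and hand them to the sidebar navigation view model.
; Profile navigation chooses party, local player, friend request or other player destinations from item type, user, current tag and the Epic parties CVar.
; Another helper filters the local user from social user arrays.
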(fn ConvertSocialUserArrayWithoutSelf
  (params (Social%20User%20VM :type Array%20of%20Fort%20Social%20User%20Viewmodel%20Object%20References))
  (for _array_element Social%20User%20VM
    (if (not (Class|FortSocialUser|IsLocalUserforBlueprint (Class|FortSocialUserViewmodel|GetSocialUser _array_element)))
      (Utilities|Array|Add (LocalVariables|GetSocialUsersWithoutSelf) (Class|FortSocialUserViewmodel|GetSocialUser _array_element))))
  (return (LocalVariables|GetSocialUsersWithoutSelf)))

(fn isUseEpicParties
  (bind _returnvalue (Development|GetConsoleVariableBoolValue "fort.UseEpicParties"))
  (return _returnvalue))

(fn NavigateToConfirmPanel
  (params (SelfReference :type Object%20Reference) (SidebarConfirmationVM :type Sidebar%20Confirmation%20Panel%20VM%20Object%20Reference) (SidebarNavigationVM :type Fort%20Sidebar%20Navigation%20VM%20Object%20Reference))
  (bind _returnvalue (Game|ConstructObjectfromClass "/Sidebar/Shared/Navigation/BP_NavigationPoint.BP_NavigationPoint_C" SelfReference 0 "None" SidebarConfirmationVM 0 0 "(TagName=\"UI.StateTree.TransitionTag.Sidebar.Overlay.Confirm\")"))
  (Class|FortSidebarNavigationVM|PushNavigation (Variables|GetSidebarNavigationVM) _returnvalue)
  (return))

(fn NavigateToPlayerProfile
  (params (SelfReference :type Object%20Reference) (SocialItemVM :type Fort%20Social%20Item%20VM%20Object%20Reference) (SidebarNavigationVM :type Fort%20Sidebar%20Navigation%20VM%20Object%20Reference) (Sidebar%20VM :type Fort%20Sidebar%20VM%20Object%20Reference))
  (bind _targetuser (Variables|/Script/SidebarUI.FortSocialItemVM|GetTargetUser SocialItemVM))
  (bind _returnvalue (Class|FortSocialUser|IsLocalUserforBlueprint _targetuser))
  (bind _returnvalue_2 (Class|FortSidebarNavigationVM|GetCurrentNavigationPoint SidebarNavigationVM))
  (LocalVariables|SetCurrentNavTag (Variables|/Script/SidebarUI.FortSidebarNavigationPoint|GetNavigationTag _returnvalue_2))
  (Utilities|IsValid (Variables|/Script/SidebarUI.FortSocialItemVM|GetTargetUser SocialItemVM)
    (:Is%20Valid
      (if (and (Math|Boolean|ORBoolean (Utilities|Enum|Equal(Enum) (Variables|/Script/SidebarUI.FortSocialItemVM|GetHeaderItemType SocialItemVM) "JoinableParties") (Utilities|Enum|Equal(Enum) (Variables|/Script/SidebarUI.FortSocialItemVM|GetHeaderItemType SocialItemVM) "JoinableNativeEpicParties") (Utilities|Enum|Equal(Enum) (Variables|/Script/SidebarUI.FortSocialItemVM|GetHeaderItemType SocialItemVM) "NativeEpicPartyInvites") (Utilities|Enum|Equal(Enum) (Variables|/Script/SidebarUI.FortSocialItemVM|GetHeaderItemType SocialItemVM) "RequestsToJoinEpicParties")) (not _returnvalue))
        _
        (elif _returnvalue
          (if (CallFunction|IsUseEpicParties 0)
            _
            (elif (== (LocalVariables|GetCurrentNavTag) "(TagName=\"UI.StateTree.TransitionTag.Sidebar.Social\")")
              ;[1 1 1] If local player select a tab and transition to `LocalPlayerTab`
              (Class|FortSidebarVM|SetFeature (Variables|GetSidebar%20VM) (LocalVariables|GetTabConfigName))
              (bind _returnvalue_3 (Game|ConstructObjectfromClass "/Sidebar/Shared/Navigation/BP_NavigationPoint.BP_NavigationPoint_C" (Variables|GetSelfReference) (Variables|/Script/SidebarUI.FortSocialItemVM|GetTargetUser (Variables|GetSocialItemVM)) "None" 0 0 (Variables|GetSocialItemVM) "(TagName=\"UI.StateTree.TransitionTag.Sidebar.LocalPlayerTab\")"))
              (Class|FortSidebarNavigationVM|SidewaysNavigation (Variables|GetSidebarNavigationVM) _returnvalue_3)))
          ; local player
          (bind _returnvalue_4 (Game|ConstructObjectfromClass "/Sidebar/Shared/Navigation/BP_NavigationPoint.BP_NavigationPoint_C" (Variables|GetSelfReference) (Variables|/Script/SidebarUI.FortSocialItemVM|GetTargetUser (Variables|GetSocialItemVM)) "None" 0 0 (Variables|GetSocialItemVM) "(TagName=\"UI.StateTree.TransitionTag.Sidebar.Social.LocalPlayer\")"))
          (Class|FortSidebarNavigationVM|PushNavigation (Variables|GetSidebarNavigationVM) _returnvalue_4)
          (else
            (bind _friendrequests (Variables|/Script/SidebarUI.FortSocialItemVM|GetFriendRequests (Variables|GetSocialItemVM))
              (:then
                (if (> (Class|FortFriendRequestsVM|GetNumRequesterUsers _friendrequests) 0)
                  ;[1 1 1] Friend Requests panel
                  (bind _returnvalue_5 (Game|ConstructObjectfromClass "/Sidebar/Shared/Navigation/BP_NavigationPoint.BP_NavigationPoint_C" (Variables|GetSelfReference) (Variables|/Script/SidebarUI.FortSocialItemVM|GetTargetUser (Variables|GetSocialItemVM)) "None" 0 0 (Variables|GetSocialItemVM) "(TagName=\"UI.StateTree.TransitionTag.Sidebar.FriendRequests\")"))
                  (Class|FortSidebarNavigationVM|PushNavigation (Variables|GetSidebarNavigationVM) _returnvalue_5)
                  (else
                    _)))
              (:else
                _))
            ;[1 1 1] If not local player populate other player tab in social section of side bar
            (bind _returnvalue_6 (Game|ConstructObjectfromClass "/Sidebar/Shared/Navigation/BP_NavigationPoint.BP_NavigationPoint_C" (Variables|GetSelfReference) (Variables|/Script/SidebarUI.FortSocialItemVM|GetTargetUser (Variables|GetSocialItemVM)) "None" 0 0 (Variables|GetSocialItemVM) "(TagName=\"UI.StateTree.TransitionTag.Sidebar.Social.OtherPlayer\")"))
            (Class|FortSidebarNavigationVM|PushNavigation (Variables|GetSidebarNavigationVM) _returnvalue_6)))))
    (:Is%20Not%20Valid
      _))
  (bind _returnvalue_7 (Game|ConstructObjectfromClass "/Sidebar/Shared/Navigation/BP_NavigationPoint.BP_NavigationPoint_C" (Variables|GetSelfReference) 0 "None" 0 0 (Variables|GetSocialItemVM) "(TagName=\"UI.StateTree.TransitionTag.Sidebar.Social.Party\")"))
  (Class|FortSidebarNavigationVM|PushNavigation (Variables|GetSidebarNavigationVM) _returnvalue_7))
