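# Sidebar Back button widget derived from WBP_Sidebar_Button_Base, showing an icon beside a CommonActionWidget prompt for the Back input action.
# PreConstruct calls SetVisuals, which creates a dynamic material from the Icon brush and sets a rotated MSDF chevron through its texture and rotation parameters.
# Construct applies the minimum button dimensions from variables.
# CF_CompactUI helper functions choose margins or size by compact-UI flag or touch input, but nothing in these graphs calls them.
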WidgetBlueprint /Sidebar/Buttons/WBP_Sidebar_Back_Button
graphs [
 CF_CompactUI_DesiredMargin {nodes [
   n1 = CF_CompactUI_DesiredMargin!() -> (
    then
    IsUsingCompactUI:bool=~
    Margin_Compact_L:double=~
    Margin_Compact_T:double=~
    Margin_Compact_R:double=~
    Margin_Compact_B:double=~
    Margin_L:double=~
    Margin_T:double=~
    Margin_R:double=~
    Margin_B:double=~
   ) {is=title K2Node_FunctionEntry}
   > n2 = Branch(Condition=@n1)
   > Return Node!(
    execute
    DesiredMargin:Margin Structure=Make Margin(Left=@n1.Margin_Compact_L, Top=@n1.Margin_Compact_T, Right=@n1.Margin_Compact_R, Bottom=@n1.Margin_Compact_B)
   )
   Return Node(execute=@n2.else, DesiredMargin=Make Margin(Left=@n1.Margin_L, Top=@n1.Margin_T, Right=@n1.Margin_R, Bottom=@n1.Margin_B))
  ]}
 CF_CompactUI_DesiredMarginForInputType {nodes [
   n1 = CF_CompactUI_DesiredMarginForInputType!() -> (
    then
    InputType:ECommonInputType Enum
    Margin_Touch_L:double=~
    Margin_Touch_T:double=~
    Margin_Touch_R:double=~
    Margin_Touch_B:double=~
    Margin_L:double=~
    Margin_T:double=~
    Margin_R:double=~
    Margin_B:double=~
   ) {is=title K2Node_FunctionEntry}
   > n2 = Branch(Condition="Equal (Enum)"(A=@n1, B=Touch))
   > Return Node(DesiredMargin=Make Margin(Left=@n1.Margin_Touch_L, Top=@n1.Margin_Touch_T, Right=@n1.Margin_Touch_R, Bottom=@n1.Margin_Touch_B))
   Return Node(execute=@n2.else, DesiredMargin=Make Margin(Left=@n1.Margin_L, Top=@n1.Margin_T, Right=@n1.Margin_R, Bottom=@n1.Margin_B))
  ]}
 CF_CompactUI_DesiredSize {nodes [
   n1 = CF_CompactUI_DesiredSize!() -> (then, IsUsingCompactUI:bool=~, Size_Compact:Vector 2D Structure, Size:Vector 2D Structure) {is=title K2Node_FunctionEntry}
   > n2 = Branch(Condition=@n1)
   > Return Node!(execute, DesiredSize:Vector 2D Structure=@n1.Size_Compact)
   Return Node(execute=@n2.else, DesiredSize=@n1.Size)
  ]}
 EventGraph {nodes [
   n1 = Event PreConstruct()
   > Parent: PreConstruct(IsDesignTime=@n1.IsDesignTime)
   > SetVisuals()
   Event Construct()
   > SetMinDimensions!(
    execute
    self:Common Button Base Object Reference
    InMinWidth:int=Get ButtonMinWidth() -> (ButtonMinWidth:int=0)
    InMinHeight:int=Get ButtonMinHeight() -> (ButtonMinHeight:int=0)
   ) -> (then)
  ]}
 SetVisuals {nodes [
   SetVisuals() {is=title K2Node_FunctionEntry}
   n1 = Get Icon() -> (Icon:Image Object Reference)
   > n2 = Set IconDynamicMaterial(IconDynamicMaterial:Material Instance Dynamic Object Reference=GetDynamicMaterial!(Brush:"Slate Brush Structure (by ref)"=Get Brush!(self:Image Object Reference=@n1) -> (Brush:Slate Brush Structure)) -> (ReturnValue)) -> (Output_Get:Material Instance Dynamic Object Reference)
   > SetTextureParameterValue(self=@n2, ParameterName=Icon, Value=/FortUILibrary/Textures/Icons/T_UI_Icon_UIKit_Chevron_MSDF)
   > SetScalarParameterValue(self=@n2, ParameterName=Rotation, Value=0.5)
  ]}
]
variables [
 ButtonMinHeight {type=integer, default=52}
 ButtonMinWidth {type=integer, default=84}
]
widgets {
 info {
  inheritedWidgetCount=0
  namedSlotCount=1
  parentClass=&/WBP_Sidebar_Button_Base.WBP_Sidebar_Button_Base_C
  rootWidgetClass=null
  widgetCount=3
 }
 widgets [
  HorizontalBox {widgetClassPath=&HorizontalBox, widget=HorizontalBox, properties {bOverride_Cursor=true}}
  Icon {
   widgetClassPath=&Image
   bIsVariable=true
   parent=HorizontalBox
   widget=Icon
   properties {brush {
     bIsDynamicallyLoaded=false
     drawAs=Image
     imageType=FullColor
     mirroring=NoMirror
     resourceName=None
     resourceObject=/Sidebar/Materials/Icons/MI_Sidebar_Icon_Base
     tiling=NoTile
     imageSize=<24 24>
     margin=<ltrb 0 0 0 0>
     outlineSettings {
      bUseBrushTransparency=false
      roundingType=HalfHeightRadius
      width=0
      color {colorUseRule=UseColor_Specified, specifiedColor=<rgba 0 0 0 0>}
      cornerRadii=<0 0 0 0>
     }
     tintColor {colorUseRule=UseColor_Specified, specifiedColor=<rgba 1 1 1 1>}
     uVRegion {bIsValid=false, max=<0 0>, min=<0 0>}
    }}
   slot {class=HorizontalBoxSlot, properties {horizontalAlignment=HAlign_Center, verticalAlignment=VAlign_Center, padding=<ltrb 10 0 6 0>}}
  }
  CommonActionWidget {
   widgetClassPath=&CommonActionWidget
   bIsVariable=true
   parent=HorizontalBox
   widget=CommonActionWidget
   properties {
    designTimeKey=Gamepad_FaceButton_Right
    inputActions [{dataTable=&InputActionDataTable, rowName=Back}]
    progressMaterialBrush {
     bIsDynamicallyLoaded=false
     drawAs=Image
     imageType=NoImage
     mirroring=NoMirror
     resourceName=None
     resourceObject=null
     tiling=NoTile
     imageSize=<32 36>
     margin=<ltrb 0 0 0 0>
     outlineSettings {
      bUseBrushTransparency=false
      roundingType=HalfHeightRadius
      width=0
      color {colorUseRule=UseColor_Specified, specifiedColor=<rgba 0 0 0 0>}
      cornerRadii=<0 0 0 0>
     }
     tintColor {colorUseRule=UseColor_Specified, specifiedColor=<rgba 1 1 1 1>}
     uVRegion {bIsValid=false, max=<0 0>, min=<0 0>}
    }
   }
   slot {class=HorizontalBoxSlot, properties {horizontalAlignment=HAlign_Center, verticalAlignment=VAlign_Center, padding=<ltrb 0 8 0 8>}}
  }
 ]
}
