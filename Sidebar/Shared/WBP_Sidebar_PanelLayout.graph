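# Shared sidebar panel layout widget with a back button, heading, subheading and named slots for header and main content.
# PreConstruct applies text, visibility and grid preview at design time only, then sets chat-view styling and compact-UI padding.
# SetActiveState acts as a switcher, collapsing all but the content or loading spinner and broadcasting the change.
# Back clicks optionally pop navigation through the sidebar navigation view model, then fire OnBackButtonClicked.
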WidgetBlueprint /Sidebar/Shared/WBP_Sidebar_PanelLayout
graphs [
 DesignTimePreset {nodes [
   Inputs()
   > SetVisibility(
    self=Get O_Back_Button() -> (O_Back_Button:Overlay Object Reference)
    InVisibility=Select(Option 0=Collapsed, Option 1=Visible, Index=Get ShowBackButton() -> (ShowBackButton:bool))
   )
   > Set Text(Text=Get HeadingText() -> (HeadingText:Text), self=Get Text_Heading() -> (Text_Heading:Common Text Object Reference))
   > Set Text(Text=Get SubheadingText() -> (SubheadingText:Text), self=Get Text_Subheading() -> (Text_Subheading:Common Text Object Reference))
   > SetGridLayout!(execute, self, HeadingText:Text=Get HeadingText()) -> (then)
   > SetActiveState!(
    execute
    self
    State:Widget Object Reference=GetActiveWidgetFromAsyncStatus!(self, IsAsyncActionOngoing:bool=Get ShowAsyncActionSpinner() -> (ShowAsyncActionSpinner:bool)) -> (Widget:Widget Object Reference)
   ) -> (then)
   n1 = Get Slot_Header() -> (Slot_Header:Named Slot Object Reference)
   > SetVisibility(
    self=@n1
    InVisibility=Select(Option 0=Collapsed, Option 1=SelfHitTestInvisible, Index=HasAnyChildren!(self:Panel Widget Object Reference=@n1) -> (ReturnValue:bool))
   )
   > Outputs!(then:Exec)
  ]}
 EventGraph {nodes [
   "On Clicked (Button_Back)"!() -> (OutputDelegate, then, Button:Common Button Base Object Reference) {is=title K2Node_ComponentBoundEvent}
   > n1 = Branch(Condition=Get AutoPopNavigationOnBack() -> (AutoPopNavigationOnBack:bool))
   > n2 = PopNavigation(self:Fort Sidebar Navigation VM Object Reference=Get FortSidebarNavigationVM() -> (FortSidebarNavigationVM:Fort Sidebar Navigation VM Object Reference))
   Call On Back Button Clicked!(execute:Exec=[@n2.then, @n1.else], self:WBP Sidebar Panel Layout Object Reference) -> (then:Exec) {is=title K2Node_CallDelegate}
   n3 = Event PreConstruct()
   > n4 = Branch(Condition=@n3.IsDesignTime)
   > n5 = Design Time Preset!(execute:Exec) -> (then:Exec) {is=title K2Node_MacroInstance}
   SetActiveState(
    execute=[@n4.else, @n5.then]
    State=GetActiveWidgetFromAsyncStatus(IsAsyncActionOngoing=Get StartInAsyncActionSpinner() -> (StartInAsyncActionSpinner:bool))
   )
   > Set Visuals!(execute:Exec) -> (then:Exec) {is=title K2Node_MacroInstance}
   > SetCompactUIStyle() {is=title K2Node_CallFunction}
  ]}
 GetActiveWidgetFromAsyncStatus {nodes [
   n1 = GetActiveWidgetFromAsyncStatus!() -> (then, IsAsyncActionOngoing:bool=~) {is=title K2Node_FunctionEntry}
   > Return Node!(
    execute
    Widget:Widget Object Reference=Select(
     Option 0:Widget Object Reference=Get List_Clipping() -> (List_Clipping:Grid Panel Object Reference)
     Option 1:Widget Object Reference=Get HB_Spinner() -> (HB_Spinner:Horizontal Box Object Reference)
     Index=@n1
    ) -> (ReturnValue:Widget Object Reference)
   )
  ]}
 GetDividerVisibility {nodes [
   n1 = GetDividerVisibility!() -> (then, HeadingText:Text) {is=title K2Node_FunctionEntry}
   > Return Node!(
    execute
    Visibility:ESlateVisibility Enum=Select(Option 0:ESlateVisibility Enum=SelfHitTestInvisible, Option 1:ESlateVisibility Enum=Collapsed, Index=TextIsEmpty(InText=@n1)) -> (ReturnValue:ESlateVisibility Enum)
   )
  ]}
 OnBackButtonClicked {nodes [OnBackButtonClicked() {is=title K2Node_FunctionEntry}]}
 OnLoadingSwitcherActiveWidgetChanged {nodes [OnLoadingSwitcherActiveWidgetChanged!() -> (then, IsAsyncActionInProgress:bool=~) {is=title K2Node_FunctionEntry}]}
 SetActiveState {nodes [
   n1 = SetActiveState!() -> (then, State:Widget Object Reference) {is=title K2Node_FunctionEntry}
   > n2 = For Each Loop(Array:Array of Widget Object References=GetAllChildren(self=Get Overlay_Main() -> (Overlay_Main:Overlay Object Reference))) -> (Array Element:Widget Object Reference)
   Call On Loading Switcher Active Widget Changed!(execute:Exec=@n2.Completed, self:WBP Sidebar Panel Layout Object Reference, IsAsyncActionInProgress:bool=Get ShowAsyncActionSpinner()) -> (then:Exec) {is=title K2Node_CallDelegate}
   SetVisibility(execute=@n2.LoopBody, self=@n2, InVisibility=Select(Option 0=Collapsed, Option 1=SelfHitTestInvisible, Index="Equal (Object)"(A=@n2, B=@n1)))
  ]}
 SetCompactUIStyle {nodes [
   SetCompactUIStyle() {is=title K2Node_FunctionEntry}
   > Branch(Condition=IsUsingCompactUI())
   n1 = Get GP_Header() -> (GP_Header:Grid Panel Object Reference)
   n2 = Get!(self:Grid Slot Object Reference=SlotAsGridSlot(Widget=@n1)) -> (Padding_Left:float=~, Padding_Top:float=~, Padding_Right:float=~, Padding_Bottom:float=~)
   > Set Grid Slot Padding!(
    execute:Exec
    Widget:Widget Object Reference=@n1
    Padding_Left:float=40
    Padding_Top:float=@n2.Padding_Top
    Padding_Right:float=40
    Padding_Bottom:float=@n2.Padding_Bottom
   ) -> (then:Exec) {is=title K2Node_MacroInstance}
  ]}
 SetGridLayout {nodes [
   n1 = SetGridLayout!() -> (then, HeadingText:Text) {is=title K2Node_FunctionEntry}
   > n2 = Cast To GridSlot!(execute, Object=Get Slot!(self:Widget Object Reference=Get Overlay_Main()) -> (Slot:Panel Slot Object Reference)) -> (then, CastFailed, AsGrid Slot:Grid Slot Object Reference) {is=title K2Node_DynamicCast}
   > n3 = Set IsTitleEmpty(IsTitleEmpty:bool=TextIsEmpty(InText=@n1)) -> (Output_Get:bool)
   > SetRow!(execute, self:Grid Slot Object Reference=@n2, InRow:int=SelectInt(A=0, B=1, bPickA=@n3)) -> (then)
   > SetRowSpan!(execute, self:Grid Slot Object Reference=@n2, InRowSpan:int=SelectInt(A=2, B=0, bPickA=@n3)) -> (then)
  ]}
 SetVisuals {nodes [
   Inputs()
   n1 = Get Backplate_ZIndexFix() -> (Backplate_ZIndexFix:WBP Sidebar Block Backplate Object Reference)
   > SetCornerRadius!(execute, self:WBP Sidebar Block Backplate Object Reference=@n1, Radius In Pixels:bool=true, Corner Radius:double=15) -> (then)
   > SetBackplateColor!(execute, self:WBP Sidebar Block Backplate Object Reference=@n1, Backplate Switch:E_UI_Sidebar_Button_Backplate_Color Enum=NewEnumerator11) -> (then)
   > SetScrollingEnabled!(
    execute
    self:Common Text Object Reference=Get Text_Heading()
    bInIsScrollingEnabled:bool=Get is Panel Name Scrolling Enabled() -> (is Panel Name Scrolling Enabled:bool)
   ) -> (then)
   > SetGridSlotAlignment!(
    execute
    Widget:Widget Object Reference=Get VB_Panel_Info() -> (VB_Panel_Info:Vertical Box Object Reference)
    Horizontal:EHorizontalAlignment Enum=HAlign_Fill
    Vertical:EVerticalAlignment Enum=Select(Option 0:EVerticalAlignment Enum=VAlign_Fill, Option 1:EVerticalAlignment Enum=VAlign_Center, Index=Get IsChatView() -> (IsChatView:bool)) -> (ReturnValue:EVerticalAlignment Enum)
   ) -> (then)
   > SetVisibility(
    self=Get Divider() -> (Divider:Image Object Reference)
    InVisibility=Select(Option 0:ESlateVisibility Enum=Collapsed, Option 1:ESlateVisibility Enum=SelfHitTestInvisible, Index=Get IsChatView()) -> (ReturnValue:ESlateVisibility Enum)
   )
   > SetGridSlotPadding!(
    execute
    Widget:Widget Object Reference=Get GP_Header()
    Padding_Left:float=36
    Padding_Top:float=20
    Padding_Right:float=36
    Padding_Bottom:float=Select(Option 0:float=6, Option 1:float=20, Index=Get IsChatView()) -> (ReturnValue:float)
   ) -> (then)
   > SetGridSlotAlignment(
    Widget=Get O_Back_Button()
    Horizontal=HAlign_Left
    Vertical=Select(Option 0:EVerticalAlignment Enum=VAlign_Top, Option 1:EVerticalAlignment Enum=VAlign_Center, Index=Get IsChatView()) -> (ReturnValue:EVerticalAlignment Enum)
   )
   > SetSize!(
    execute
    self:Spacer Object Reference=Get Text_Heading_Spacer() -> (Text_Heading_Spacer:Spacer Object Reference)
    InSize_X:double=0
    InSize_Y:double=Select(Option 0:double=52, Option 1:double=32, Index=Get IsChatView()) -> (ReturnValue:double)
   ) -> (then)
   > Outputs()
  ]}
]
variables [
 AutoPopNavigationOnBack {type=boolean, default=true}
 HeadingText {type=string, default=Panel Name Extra Long Gets Ellipses}
 is Panel Name Scrolling Enabled {type=boolean}
 IsChatView {type=boolean}
 OnBackButtonClicked {}
 OnLoadingSwitcherActiveWidgetChanged {}
 ShowAsyncActionSpinner {type=boolean}
 ShowBackButton {type=boolean, default=true}
 StartInAsyncActionSpinner {type=boolean}
 SubheadingText {type=string, default=""}
]
widgets {
 info {inheritedWidgetCount=0, parentClass=&UserWidget, rootWidgetClass=&GridPanel, widgetCount=20}
 widgets [
  Root {widgetClassPath=&GridPanel, widget=Root, properties {bOverride_Cursor=true, columnFill [1], rowFill [1]}}
  Panel_Wrapper {
   widgetClassPath=&GridPanel
   parent=Root
   widget=Panel_Wrapper
   properties {bOverride_Cursor=true, columnFill [1], rowFill [0,1]}
   slot {class=GridSlot}
  }
  GP_Header {
   widgetClassPath=&GridPanel
   bIsVariable=true
   parent=Panel_Wrapper
   widget=GP_Header
   properties {bOverride_Cursor=true, columnFill [0,1,0], rowFill [0]}
   slot {class=GridSlot, properties {layer=10, padding=<ltrb 36 20 36 6>}}
  }
  O_Back_Button {
   widgetClassPath=&Overlay
   bIsVariable=true
   parent=GP_Header
   widget=O_Back_Button
   slot {class=GridSlot, properties {columnSpan=1, horizontalAlignment=HAlign_Left, layer=1, verticalAlignment=VAlign_Top, padding=<ltrb 0 0 24 0>}}
  }
  Backplate_ZIndexFix {
   widgetClassPath=/Sidebar/Buttons/Blocks/WBP_Sidebar_Block_Backplate
   bIsVariable=true
   parent=O_Back_Button
   widget=Backplate_ZIndexFix
   slot {class=OverlaySlot, properties {horizontalAlignment=HAlign_Fill, verticalAlignment=VAlign_Fill}}
  }
  Button_Back {
   widgetClassPath=/Sidebar/Buttons/WBP_Sidebar_Back_Button
   bIsVariable=true
   parent=O_Back_Button
   widget=Button_Back
   properties {
    bIsFocusable=false
    holdData=&CommonUI_HoldData_Base_C
    hoveredSlateSoundOverride {resourceObject=/Game/Sounds/Fort_UI_Sounds/FrontEnd/DefaultSounds/PressedButtons/MSS_UI_Button_Hover_Default}
    pressedSlateSoundOverride {resourceObject=/Game/Sounds/Fort_UI_Sounds/FrontEnd/DefaultSounds/PressedButtons/MSS_UI_Button_Pressed_Default}
   }
   slot {class=OverlaySlot, properties {horizontalAlignment=HAlign_Fill, verticalAlignment=VAlign_Fill}}
  }
  HorizontalBox_0 {
   widgetClassPath=&HorizontalBox
   parent=GP_Header
   widget=HorizontalBox_0
   slot {class=GridSlot, properties {column=1, columnSpan=1, verticalAlignment=VAlign_Center}}
  }
  Slot_Header {
   widgetClassPath=&NamedSlot
   bIsVariable=true
   parent=HorizontalBox_0
   widget=Slot_Header
   slot {class=HorizontalBoxSlot, properties {padding=<ltrb -10 0 14 0>}}
  }
  VB_Panel_Info {
   widgetClassPath=&VerticalBox
   bIsVariable=true
   parent=HorizontalBox_0
   widget=VB_Panel_Info
   properties {clipping=ClipToBounds}
   slot {class=HorizontalBoxSlot, properties {verticalAlignment=VAlign_Center, size {sizeRule=Fill, value=1}}}
  }
  Overlay_0 {
   widgetClassPath=&Overlay
   parent=VB_Panel_Info
   widget=Overlay_0
   slot {class=VerticalBoxSlot, properties {verticalAlignment=VAlign_Center, size {sizeRule=Fill, value=1}}}
  }
  Text_Heading {
   widgetClassPath=&CommonTextBlock
   bIsVariable=true
   parent=Overlay_0
   widget=Text_Heading
   properties {
    bAutoCollapseWithEmptyText=true
    bIsScrollingEnabled=false
    mobileFontSizeMultiplier=1.22
    scrollStyle=&TextScrollStyle-Base_C
    text=Panel extra long text get ellilpses
    textOverflowPolicy=Ellipsis
    font {
     bForceMonospaced=false
     bMaterialIsStencil=false
     fontMaterial=/Sidebar/Materials/Fonts/MI_Sidebar_FontColor_Primary
     fontObject=&HeadingNow
     letterSpacing=0
     monospacedWidth=1
     size=27
     skewAmount=0
     typefaceFontName=75 Medium
     outlineSettings {
      bApplyOutlineToDropShadows=false
      bMiteredCorners=false
      bSeparateFillAlpha=false
      outlineMaterial=null
      outlineSize=0
      outlineColor=<rgba 0 0 0 1>
     }
    }
    shadowOffset=<0 0>
   }
   slot {class=OverlaySlot, properties {verticalAlignment=VAlign_Center}}
  }
  Text_Heading_Spacer {
   widgetClassPath=&Spacer
   bIsVariable=true
   parent=Overlay_0
   widget=Text_Heading_Spacer
   properties {size=<0 52>}
   slot {class=OverlaySlot, properties {horizontalAlignment=HAlign_Fill, verticalAlignment=VAlign_Fill}}
  }
  Text_Subheading {
   widgetClassPath=&CommonTextBlock
   bIsVariable=true
   parent=VB_Panel_Info
   widget=Text_Subheading
   properties {
    bAutoCollapseWithEmptyText=true
    mobileFontSizeMultiplier=1.41
    text=Subheading
    textOverflowPolicy=Ellipsis
    font {
     bForceMonospaced=false
     bMaterialIsStencil=false
     fontMaterial=/Sidebar/Materials/Fonts/MI_Sidebar_FontColor_Secondary
     fontObject=&HeadingNow
     letterSpacing=0
     monospacedWidth=1
     size=12.75
     skewAmount=0
     typefaceFontName=74 Regular
     outlineSettings {
      bApplyOutlineToDropShadows=false
      bMiteredCorners=false
      bSeparateFillAlpha=false
      outlineMaterial=null
      outlineSize=0
      outlineColor=<rgba 0 0 0 1>
     }
    }
    shadowOffset=<0 0>
   }
   slot {class=VerticalBoxSlot, properties {verticalAlignment=VAlign_Center}}
  }
  Divider {
   widgetClassPath=&Image
   bIsVariable=true
   parent=Panel_Wrapper
   widget=Divider
   properties {brush {
     bIsDynamicallyLoaded=false
     drawAs=Image
     imageType=FullColor
     mirroring=NoMirror
     resourceName=None
     resourceObject=/Sidebar/Materials/MI_Sidebar_Divider
     tiling=NoTile
     imageSize=<32 2>
     margin=<ltrb 0 0 0 0>
     outlineSettings {
      bUseBrushTransparency=false
      roundingType=HalfHeightRadius
      width=0
      color {colorUseRule=UseColor_Specified, specifiedColor=<rgba 0 0 0 0>}
      cornerRadii=<0 0 0 0>
     }
     tintColor {colorUseRule=UseColor_Specified, specifiedColor=<rgba 1 1 1 1>}
     uVRegion {bIsValid=false, max=<0 0>, min=<0 0>}
    }}
   slot {class=GridSlot, properties {verticalAlignment=VAlign_Bottom}}
  }
  Overlay_Main {
   widgetClassPath=&Overlay
   bIsVariable=true
   parent=Panel_Wrapper
   widget=Overlay_Main
   slot {class=GridSlot, properties {row=1, padding=<ltrb 0 3 0 0>}}
  }
  HB_Spinner {
   widgetClassPath=&HorizontalBox
   bIsVariable=true
   parent=Overlay_Main
   widget=HB_Spinner
   slot {class=OverlaySlot, properties {horizontalAlignment=HAlign_Center}}
  }
  Throbber {
   widgetClassPath=/FortUILibrary/UIKit/Throbber/WBP_UIKit_Throbber
   parent=HB_Spinner
   widget=Throbber
   slot {class=HorizontalBoxSlot, properties {padding=<ltrb 0 0 12 0>}}
  }
  Text_Loading {
   widgetClassPath=&CommonTextBlock
   parent=HB_Spinner
   widget=Text_Loading
   properties {
    bAutoCollapseWithEmptyText=true
    mobileFontSizeMultiplier=1.2
    text=Loading...
    font {
     bForceMonospaced=false
     bMaterialIsStencil=false
     fontMaterial=null
     fontObject=&HeadingNow
     letterSpacing=0
     monospacedWidth=1
     size=15.75
     skewAmount=0
     typefaceFontName=75 Medium
     outlineSettings {
      bApplyOutlineToDropShadows=false
      bMiteredCorners=false
      bSeparateFillAlpha=false
      outlineMaterial=null
      outlineSize=0
      outlineColor=<rgba 0 0 0 1>
     }
    }
    shadowOffset=<0 0>
   }
   slot {class=HorizontalBoxSlot, properties {verticalAlignment=VAlign_Center}}
  }
  List_Clipping {
   widgetClassPath=&GridPanel
   bIsVariable=true
   parent=Overlay_Main
   widget=List_Clipping
   properties {clipping=ClipToBounds, columnFill [1], rowFill [1]}
   slot {class=OverlaySlot, properties {horizontalAlignment=HAlign_Fill, verticalAlignment=VAlign_Fill}}
  }
  Slot_Main {
   widgetClassPath=&NamedSlot
   bIsVariable=true
   parent=List_Clipping
   widget=Slot_Main
   slot {class=GridSlot, properties {padding=<ltrb 18 0 18 18>}}
  }
 ]
}
