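# Shared sidebar user row widget showing an avatar, rich-text name, status message and mutual-friends line, with named slots for extra content.
# PreConstruct caps name width, swaps the avatar for any override-slot content and enlarges it in compact UI.
# Helper functions build name markup from nickname, player name and platform icon, redden secondary text for blocked users, and toggle text scrolling.
# Mouse enter and leave broadcast hover delegates.
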WidgetBlueprint /Sidebar/Shared/WBP_Sidebar_User
graphs [
 EventGraph {nodes [
   Event PreConstruct()
   > SetMaxTextWidthBeforeClipping!(execute, self, MaxNameTextWidth:double=Get Text_MaxWidth() -> (Text_MaxWidth:double)) -> (then)
   n1 = HasAnyChildren!(self:Panel Widget Object Reference=Get Slot_AvatarOverride() -> (Slot_AvatarOverride:Named Slot Object Reference)) -> (ReturnValue:bool)
   > SetVisibility(
    self=Get PlayerAvatar() -> (PlayerAvatar:WBP Sidebar Player Avatar Object Reference)
    InVisibility=Select(Option 0=SelfHitTestInvisible, Option 1=Collapsed, Index=@n1)
   )
   > SetVisibility(self=Get Slot_AvatarOverride(), InVisibility=Select(Option 0=Collapsed, Option 1=SelfHitTestInvisible, Index=@n1))
   > SetBlockedTextColor!(execute, self, IsUserBlocked:bool=false) -> (then)
   > SetCompactUIStyle() {is=title K2Node_CallFunction}
   Event OnMouseEnter()
   > Call On Hovered!(execute:Exec, self:WBP Sidebar User Object Reference) -> (then:Exec) {is=title K2Node_CallDelegate}
   Event OnMouseLeave!() -> (OutputDelegate, then, MouseEvent:"Pointer Event Structure (by ref)") {is=title K2Node_Event}
   > Call On Unhovered!(execute:Exec, self:WBP Sidebar User Object Reference) -> (then:Exec) {is=title K2Node_CallDelegate}
  ]}
 GetConcatenatedNameAndNicknameWithIcon {nodes [
   GetConcatenatedNameAndNicknameWithIcon!() -> (then, NickName:String, PlayerName:String, PlatformIconTag:String) {is=title K2Node_FunctionEntry}
   n1 = Get NickName() -> (NickName:String)
   n2 = Get PlatformIconTag() -> (PlatformIconTag:String)
   > n3 = Branch(Condition=AND Boolean(A=IsEmpty(InString=@n1), B=IsEmpty(InString=@n2)))
   > Return Node!(execute, FinalNameText:Text="To Text (String)"(InString=Get PlayerName() -> (PlayerName:String)))
   n4 = Branch(execute=@n3.else, Condition=IsEmpty(InString=Get NickName()))
   > Return Node(FinalNameText=Format Text!(Format="INVTEXT(\"{PlayerName}<{platformIcon}/>\")", PlayerName:String=Get PlayerName(), platformIcon:String=Get PlatformIconTag()) -> (Result))
   n5 = Branch(execute=@n4.else, Condition=IsEmpty(InString=Get PlatformIconTag()))
   > Return Node(FinalNameText=Format Text!(Format="INVTEXT(\"{Nickname} <SecondaryNameStyle>{PlayerName}</>\")", Nickname:String=Get NickName(), PlayerName=Get PlayerName()) -> (Result))
   Return Node(
    execute=@n5.else
    FinalNameText=Format Text!(
     Format="INVTEXT(\"{Nickname}<{PlatformIcon}/><SecondaryNameStyle>{PlayerName}</>\")"
     Nickname=Get NickName()
     PlatformIcon:String=Get PlatformIconTag()
     PlayerName=Get PlayerName()
    ) -> (Result)
   )
   Format Text!(Format="INVTEXT(\"{PlayerName}\")", PlayerName=Get PlayerName()) -> (Result)
  ]}
 On Hovered {nodes [On Hovered() {is=title K2Node_FunctionEntry}]}
 On Unhovered {nodes [On Unhovered() {is=title K2Node_FunctionEntry}]}
 OnSidebarUserHovered {nodes [
   OnSidebarUserHovered() {is=title K2Node_FunctionEntry}
   > SetScrollingEnabled!(
    execute
    self:Common Text Object Reference=Get Text_Message() -> (Text_Message:Common Text Object Reference)
    bInIsScrollingEnabled:bool=true
   ) -> (then)
   > SetScrollingEnabled(
    self:Common Rich Text Block Object Reference=Get RichText_Name() -> (RichText_Name:Common Rich Text Block Object Reference)
    bInIsScrollingEnabled=true
   )
  ]}
 OnSidebarUserUnHovered {nodes [
   OnSidebarUserUnHovered() {is=title K2Node_FunctionEntry}
   > SetScrollingEnabled(self:Common Text Object Reference=Get Text_Message(), bInIsScrollingEnabled=false)
   > SetScrollingEnabled(self:Common Rich Text Block Object Reference=Get RichText_Name(), bInIsScrollingEnabled=false)
  ]}
 SetBlockedStatus {nodes [
   n1 = SetBlockedStatus!() -> (then, Interactions:Array of Fort Social Interaction VM Object References) {is=title K2Node_FunctionEntry}
   > SetBlockedTextColor(IsUserBlocked=false)
   > Branch(Condition=Is Not Empty!(TargetArray:Array of Fort Social Interaction VM Object References=@n1) -> (ReturnValue:bool))
   > n2 = For Each Loop with Break!(
    Exec:Exec
    Array:Array of Fort Social Interaction VM Object References=Get Interactions() -> (Interactions:Array of Fort Social Interaction VM Object References)
    Break:Exec=@n3.then
   ) -> (LoopBody:Exec, Array Element:Fort Social Interaction VM Object Reference, Array Index:int, Completed:Exec)
   > Branch(Condition="Equal (Enum)"(
     A:ESocialInteractionType Enum=Get InteractionType!(self:Fort Social Interaction VM Object Reference=@n2) -> (InteractionType:ESocialInteractionType Enum=AcceptEpicPartyInvite)
     B:ESocialInteractionType Enum=Unblock
    ))
   > n3 = SetBlockedTextColor(IsUserBlocked=true)
  ]}
 SetBlockedTextColor {nodes [
   n1 = SetBlockedTextColor!() -> (then, IsUserBlocked:bool=~) {is=title K2Node_FunctionEntry}
   > SetColorAndOpacity(
    self=[Get MutualFriends_Text() -> (MutualFriends_Text:Common Text Object Reference), Get Text_Message()]
    InColorAndOpacity=Make Slate Color(SpecifiedColor=SelectColor(A=(R=1,G=0.147,B=0.1329,A=1), B=(R=1,G=1,B=1,A=0.5), bPickA=@n1))
   )
  ]}
 SetCompactUIStyle {nodes [
   SetCompactUIStyle() {is=title K2Node_FunctionEntry}
   > Branch(Condition=IsUsingCompactUI())
   n1 = Get PlayerAvatar()
   > Event Set Size!(execute, self:WBP Sidebar Player Avatar Object Reference=@n1, NewSize:Vector 2D Structure=(X=88,Y=88)) -> (then)
   n2 = Get!(self:User Widget Object Reference=@n1) -> (Padding_Left:float=~, Padding_Top:float=~, Padding_Right:float=~, Padding_Bottom:float=~)
   > SetGridSlotPadding!(
    execute
    Widget:Widget Object Reference=@n1
    Padding_Left:float=@n2
    Padding_Top:float=@n2.Padding_Top
    Padding_Right:float=24
    Padding_Bottom:float=@n2.Padding_Bottom
   ) -> (then)
  ]}
 SetMaxTextWidthBeforeClipping {nodes [
   n1 = SetMaxTextWidthBeforeClipping!() -> (then, MaxNameTextWidth:double=~) {is=title K2Node_FunctionEntry}
   > n2 = Set Text_MaxWidth(Text_MaxWidth:double=@n1) -> (Output_Get:double)
   > SetMaxDesiredWidth!(
    execute
    self:Size Box Object Reference=Get SizeBox_TextClipping() -> (SizeBox_TextClipping:Size Box Object Reference)
    InMaxDesiredWidth:float=@n2
   ) -> (then)
  ]}
]
variables [
 On Hovered {}
 On Unhovered {}
 Text_MaxWidth {type=number, default=380}
]
widgets {
 info {inheritedWidgetCount=0, parentClass=&CommonUserWidget, rootWidgetClass=&GridPanel, widgetCount=12}
 widgets [
  GridPanel_18 {widgetClassPath=&GridPanel, widget=GridPanel_18, properties {columnFill [0,1], rowFill [1]}}
  VB_RightColumn {
   widgetClassPath=&VerticalBox
   parent=GridPanel_18
   widget=VB_RightColumn
   slot {class=GridSlot, properties {column=1, columnSpan=1, verticalAlignment=VAlign_Center}}
  }
  HorizontalStackBox_Name {
   widgetClassPath=&StackBox
   bIsVariable=true
   parent=VB_RightColumn
   widget=HorizontalStackBox_Name
   properties {clipping=ClipToBounds}
   slot {class=VerticalBoxSlot, properties {padding=<ltrb 0 0 0 3>}}
  }
  SizeBox_TextClipping {
   widgetClassPath=&SizeBox
   bIsVariable=true
   parent=HorizontalStackBox_Name
   widget=SizeBox_TextClipping
   properties {bOverride_MaxDesiredWidth=true, bOverride_MinDesiredHeight=true, maxDesiredWidth=380, minDesiredHeight=24}
   slot {class=StackBoxSlot, properties {verticalAlignment=VAlign_Center}}
  }
  RichText_Name {
   widgetClassPath=&CommonRichTextBlock
   bIsVariable=true
   parent=SizeBox_TextClipping
   widget=RichText_Name
   properties {
    clipping=ClipToBounds
    mobileTextBlockScale=1.3
    scrollStyle=&TextScrollStyle-Base_C
    text="Nickname<XSX/><SecondaryNameStyle>{NameExtraLongggg}</>"
    textOverflowPolicy=Ellipsis
    textStyleSet=&/DT_UI_RichText_Sidebar
   }
   slot {class=SizeBoxSlot, properties {horizontalAlignment=HAlign_Left, verticalAlignment=VAlign_Center}}
  }
  Slot_FirstLine {
   widgetClassPath=&NamedSlot
   bIsVariable=true
   parent=HorizontalStackBox_Name
   widget=Slot_FirstLine
   slot {class=StackBoxSlot, properties {padding=<ltrb 8 0 0 0>, size {sizeRule=Fill, value=1}}}
  }
  HorizontalBox_0 {
   widgetClassPath=&HorizontalBox
   parent=VB_RightColumn
   widget=HorizontalBox_0
   properties {clipping=ClipToBounds}
   slot {class=VerticalBoxSlot, properties {verticalAlignment=VAlign_Bottom}}
  }
  Slot_SecondLine {widgetClassPath=&NamedSlot, bIsVariable=true, parent=HorizontalBox_0, widget=Slot_SecondLine, slot {class=HorizontalBoxSlot}}
  Text_Message {
   widgetClassPath=&CommonTextBlock
   bIsVariable=true
   parent=HorizontalBox_0
   widget=Text_Message
   properties {
    bAutoCollapseWithEmptyText=true
    bIsScrollingEnabled=false
    clipping=ClipToBounds
    mobileFontSizeMultiplier=1.4
    scrollStyle=&TextScrollStyle-Base_C
    text="Message - Really Long Status Here, Game Status will clip and scroll when extra long"
    textOverflowPolicy=Ellipsis
    colorAndOpacity {colorUseRule=UseColor_Specified, specifiedColor=<rgba 1 1 1 0.5>}
    font {
     bForceMonospaced=false
     bMaterialIsStencil=false
     fontMaterial=/Sidebar/Materials/Fonts/MI_Sidebar_FontColor_Secondary
     fontObject=&HeadingNow
     letterSpacing=0
     monospacedWidth=1
     size=12.75
     skewAmount=0
     typefaceFontName=74 Regular
     outlineSettings {
      bApplyOutlineToDropShadows=false
      bMiteredCorners=false
      bSeparateFillAlpha=false
      outlineMaterial=/FortUILibrary/Systems/ModularButton/MaterialInstances/MI_UIKit_TextOutline
      outlineSize=0
      outlineColor=<rgba 0 0 0 1>
     }
    }
    shadowOffset=<0 0>
   }
   slot {class=HorizontalBoxSlot, properties {size {sizeRule=Fill, value=1}}}
  }
  MutualFriends_Text {
   widgetClassPath=&CommonTextBlock
   bIsVariable=true
   parent=VB_RightColumn
   widget=MutualFriends_Text
   properties {
    bAutoCollapseWithEmptyText=true
    bIsScrollingEnabled=false
    clipping=ClipToBounds
    mobileFontSizeMultiplier=1.4
    scrollStyle=&TextScrollStyle-Base_C
    text=1 Mutual Friend
    textOverflowPolicy=Ellipsis
    visibility=Collapsed
    colorAndOpacity {colorUseRule=UseColor_Specified, specifiedColor=<rgba 1 1 1 0.5>}
    font {
     bForceMonospaced=false
     bMaterialIsStencil=false
     fontMaterial=&MI_Sidebar_FontColor_Secondary
     fontObject=&HeadingNow
     letterSpacing=0
     monospacedWidth=1
     size=12.75
     skewAmount=0
     typefaceFontName=74 Regular
     outlineSettings {
      bApplyOutlineToDropShadows=false
      bMiteredCorners=false
      bSeparateFillAlpha=false
      outlineMaterial=&MI_UIKit_TextOutline
      outlineSize=0
      outlineColor=<rgba 0 0 0 1>
     }
    }
    shadowOffset=<0 0>
   }
   slot {class=VerticalBoxSlot, properties {verticalAlignment=VAlign_Bottom, padding=<ltrb 0 3 0 0>}}
  }
  PlayerAvatar {
   widgetClassPath=/Sidebar/Panels/WBP_Sidebar_PlayerAvatar
   bIsVariable=true
   parent=GridPanel_18
   widget=PlayerAvatar
   properties {avatarSize=<72 72>}
   slot {class=GridSlot, properties {verticalAlignment=VAlign_Center, padding=<ltrb 0 0 10 0>}}
  }
  Slot_AvatarOverride {
   widgetClassPath=&NamedSlot
   bIsVariable=true
   parent=GridPanel_18
   widget=Slot_AvatarOverride
   slot {class=GridSlot, properties {horizontalAlignment=HAlign_Left}}
  }
 ]
}
