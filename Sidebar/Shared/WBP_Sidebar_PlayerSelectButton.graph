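# Sidebar list-entry button, derived from WBP_Sidebar_Button_Base, that shows a social user beside a toggle for selecting players.
# When its list item is set, it feeds the FortSocialItemVM's target user to name and info view models and initializes the avatar.
# Clicking flips the item VM's bIsChecked and updates the toggle; hover events forward to SidebarUser.
# A SetCompactUIStyle override, run on PreConstruct, adjusts toggle size and padding for compact UI.
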WidgetBlueprint /Sidebar/Shared/WBP_Sidebar_PlayerSelectButton
graphs [
 EventGraph {nodes [
   Event On Clicked()
   > n1 = Get!(execute:Exec) -> (then:Exec, else:Exec, FortSocialItemVM:Fort Social Item VM Object Reference)
   > n2 = Set bIsChecked!(
    execute
    bIsChecked:bool=NOT Boolean(A=Get bIsChecked!(self:Fort Social Item VM Object Reference=@n1) -> (bIsChecked:bool))
    self:Fort Social Item VM Object Reference=@n1
   ) -> (then, Output_Get:bool)
   > SetSelected!(
    execute
    self:WBP UIKit Block Toggle Object Reference=Get Block_Toggle() -> (Block_Toggle:WBP UIKit Block Toggle Object Reference)
    Selected:bool=@n2
   ) -> (then)
   n3 = Event OnListItemObjectSet()
   > n4 = Cast To FortSocialItemVM!(execute, Object=@n3.ListItemObject) -> (then, CastFailed, AsFort Social Item VM:Fort Social Item VM Object Reference) {is=title K2Node_DynamicCast}
   > InitializeVMs!(execute, self, FortSocialItemVM:Fort Social Item VM Object Reference=@n4) -> (then)
   Event On Hovered()
   > Parent: On Hovered() {is=title K2Node_CallParentFunction}
   n5 = Get SidebarUser() -> (SidebarUser:WBP Sidebar User Object Reference)
   > OnSidebarUserHovered(self:WBP Sidebar User Object Reference=@n5)
   Event On Unhovered()
   > Parent: On Unhovered() {is=title K2Node_CallParentFunction}
   > OnSidebarUserUnHovered(self:WBP Sidebar User Object Reference=@n5)
   n6 = Event PreConstruct()
   > Parent: PreConstruct(IsDesignTime=@n6.IsDesignTime)
   > SetCompactUIStyle(self:WBP Sidebar Button Base Object Reference) {is=title K2Node_CallFunction}
  ]}
 GetNameFromSidebarUser {nodes [
   n1 = GetNameFromSidebarUser!() -> (then, NickName:String, PlayerName:String, PlatformIconTag:String) {is=title K2Node_FunctionEntry}
   > Return Node!(
    execute
    FinalNameText:Text=GetConcatenatedNameAndNicknameWithIcon!(
     self:WBP Sidebar User Object Reference=Get SidebarUser()
     NickName:String=@n1
     PlayerName:String=@n1.PlayerName
     PlatformIconTag:String=@n1.PlatformIconTag
    ) -> (FinalNameText:Text)
   )
  ]}
 InitializeVMs {nodes [
   n1 = InitializeVMs!() -> (then, FortSocialItemVM:Fort Social Item VM Object Reference) {is=title K2Node_FunctionEntry}
   > n2 = Set FortSocialItemVM(FortSocialItemVM:Fort Social Item VM Object Reference=@n1) -> (Output_Get:Fort Social Item VM Object Reference)
   > n3 = Sequence()
   > n4 = Get!(execute, self:Fort Social Item VM Object Reference=@n2) -> (then, else, TargetUser:Fort Social User Weak object)
   > SetSocialUser!(
    execute
    self:Fort Social User Viewmodel Object Reference=[
     Get FortSocialUserNameViewModel() -> (FortSocialUserNameViewModel:Fort Social User Name Viewmodel Object Reference)
     Get FortSocialUserInfoViewModel() -> (FortSocialUserInfoViewModel:Fort Social User Info View Model Object Reference)
    ]
    InSocialUser:Fort Social User Object Reference=@n4
   ) -> (then)
   InitializeWithFortSocialItemVM!(
    execute=@n3.then_1
    self:WBP Sidebar Player Avatar Object Reference=Get PlayerAvatar!(self:WBP Sidebar User Object Reference=Get SidebarUser()) -> (PlayerAvatar:WBP Sidebar Player Avatar Object Reference)
    FortSocialItemVM:Fort Social Item VM Object Reference=@n2
   ) -> (then)
  ]}
 SetCompactUIStyle {nodes [
   SetCompactUIStyle() {is=title K2Node_FunctionEntry}
   > Parent: SetCompactUIStyle() {is=title K2Node_CallParentFunction}
   > Branch(Condition=IsUsingCompactUI())
   n1 = Get Block_Toggle()
   > SetSize!(execute, self:WBP UIKit Block Toggle Object Reference=@n1, Desired Size:Vector 2D Structure=(X=40,Y=40)) -> (then)
   n2 = SlotAsHorizontalBoxSlot!(Widget:Widget Object Reference=@n1) -> (ReturnValue:Horizontal Box Slot Object Reference)
   n3 = Get!(self:Horizontal Box Slot Object Reference=@n2) -> (Padding_Left:float=~, Padding_Top:float=~, Padding_Right:float=~, Padding_Bottom:float=~)
   > SetPadding!(
    execute
    self:Horizontal Box Slot Object Reference=@n2
    InPadding_Left:float=24
    InPadding_Top:float=@n3.Padding_Top
    InPadding_Right:float=20
    InPadding_Bottom:float=@n3.Padding_Bottom
   ) -> (then)
   > SetPadding(self:User Widget Object Reference=Get SidebarUser(), InPadding_Left=20, InPadding_Top=20, InPadding_Right=0, InPadding_Bottom=20)
  ]}
]
variables []
widgets {
 info {
  inheritedWidgetCount=0
  namedSlotCount=1
  parentClass=/Sidebar/Buttons/WBP_Sidebar_Button_Base.WBP_Sidebar_Button_Base_C
  rootWidgetClass=null
  widgetCount=3
 }
 widgets [
  HorizontalBox_0 {widgetClassPath=&HorizontalBox, widget=HorizontalBox_0}
  SidebarUser {
   widgetClassPath=&/WBP_Sidebar_User
   bIsVariable=true
   parent=HorizontalBox_0
   widget=SidebarUser
   properties {padding=<ltrb 18 15 0 15>}
   slot {class=HorizontalBoxSlot, properties {size {sizeRule=Fill, value=1}}}
  }
  Block_Toggle {
   widgetClassPath=/FortUILibrary/Systems/ModularButton/Blocks/WBP_UIKit_Block_Toggle
   bIsVariable=true
   parent=HorizontalBox_0
   widget=Block_Toggle
   properties {
    toggleType=Toggle
    brush {
     bIsDynamicallyLoaded=false
     drawAs=Image
     imageType=NoImage
     mirroring=NoMirror
     resourceName=None
     resourceObject=/FortUILibrary/Systems/ModularButton/MaterialInstances/MI_UIKit_Toggle
     tiling=NoTile
     imageSize=<36 36>
     margin=<ltrb 0 0 0 0>
     outlineSettings {
      bUseBrushTransparency=false
      roundingType=HalfHeightRadius
      width=0
      color {colorUseRule=UseColor_Specified, specifiedColor=<rgba 0 0 0 0>}
      cornerRadii=<0 0 0 1>
     }
     tintColor {colorUseRule=UseColor_Specified, specifiedColor=<rgba 1 1 1 1>}
     uVRegion {bIsValid=false, max=<0 0>, min=<0 0>}
    }
   }
   slot {class=HorizontalBoxSlot, properties {verticalAlignment=VAlign_Center, padding=<ltrb 18 0 18 0>}}
  }
 ]
}
